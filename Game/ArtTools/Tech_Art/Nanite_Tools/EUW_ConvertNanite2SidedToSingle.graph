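# Editor utility widget that turns off two-sided distance field generation on selected Nanite-enabled static meshes.
# The button loops over selected assets, checks LOD 0 build settings, then either clears the flag and saves, or in query-only mode counts matches and optionally adds them to a named collection.
# Blend mode, referencer class and auto-LOD checkbox checks exist but sit on a chain unconnected to the button's flow.
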EditorUtilityWidgetBlueprint /Game/ArtTools/Tech_Art/Nanite_Tools/EUW_ConvertNanite2SidedToSingle
graphs [
 CheckForValidReferencers {nodes [
   n1 = CheckForValidReferencers!() -> (then, StaticMesh:Static Mesh Object Reference, ParentClassContains:String) {is=title K2Node_FunctionEntry}
   n2 = Get Object Path String(Object=@n1)
   n3 = GetAssetRegistry()
   > n4 = Get Referencers!(
    execute
    self:Asset Registry Interface=@n3
    PackageName:Name=String To Name(InString=Split(SourceString=@n2, InStr=., SearchDir=FromEnd))
    ReferenceOptions:"Asset Registry Dependency Options Structure (by ref)"=Make Asset Registry Dependency Options(bIncludeSoftPackageReferences=True, bIncludeSoftManagementReferences=true, bIncludeHardManagementReferences=true)
   ) -> (then, OutReferencers:Array of Names, ReturnValue:bool) {is=title K2Node_CallFunction}
   > n5 = GetAssets!(
    execute
    self:Asset Registry Interface=@n3
    Filter:"ARFilter Structure (by ref)"=MakeARFilter!(
     PackageNames:Array of Names=@n4
     PackagePaths:Array of Names
     SoftObjectPaths:Array of Soft Object Path Structures
     ClassPaths:Array of Top Level Asset Path Structures
     RecursiveClassPathsExclusionSet:Set of Top Level Asset Path Structures
     ClassNames:Array of Names
     RecursiveClassesExclusionSet:Set of Names
     bRecursivePaths:bool=false
     bRecursiveClasses:bool=false
     bIncludeOnlyOnDiskAssets:bool=false
    ) -> (ReturnValue:ARFilter Structure)
    bSkipARFilteredAssets:bool=true
   ) -> (then, OutAssetData:Array of Asset Data Structures, ReturnValue:bool)
   > n6 = For Each Loop with Break!(Exec:Exec, Array:Array of Asset Data Structures=@n5, Break:Exec=@n8.then) -> (LoopBody:Exec, Array Element:Asset Data Structure, Array Index:int, Completed:Exec)
   Return Node!(execute=@n6.Completed, DoesSM_HaveReferencerOfGivenClass?:bool=Get SafeToConvertPerReferencers() -> (SafeToConvertPerReferencers:bool))
   n7 = MakeLiteralName(Value=NativeParentClass)
   > Branch(Condition=Contains(SearchIn=GetTagValue(InAssetData=@n6, InTagName=@n7), Substring=@n1.ParentClassContains))
   > n8 = Set SafeToConvertPerReferencers(SafeToConvertPerReferencers:bool=true) -> (Output_Get:bool)
  ]}
 CheckForWPOSwitchActivation {nodes [
   n1 = CheckForWPOSwitchActivation!() -> (then, Material:Material Interface Object Reference, CheckForWPO?:bool=~) {is=title K2Node_FunctionEntry}
   > n2 = Branch(Condition=@n1.CheckForWPO?)
   > n3 = Cast To MaterialInstanceConstant(Object=@n1)
   > PrintString(InString=Valid)
   PrintString(execute=@n3.CastFailed, InString="Invalid Material Assignment!")
   Return Node!(execute=@n2.else)
   GetMaterialInstanceStaticSwitchParameterValue!(
    Instance:Material Instance Constant Object Reference=@n3
    ParameterName:Name=None
    Association:EMaterialParameterAssociation Enum=GlobalParameter
   ) -> (ReturnValue:bool)
  ]}
 EventGraph {nodes [
   "On Clicked (TA_Generic01_Button)"!() -> (OutputDelegate, then, widget:User Widget Object Reference, event:Pointer Event Structure) {is=title K2Node_ComponentBoundEvent}
   > Set NumberAffected(NumberAffected:int=0) -> (Output_Get:int=0)
   > n1 = GetSelectedAssets()
   > n2 = Branch(Condition=Get AddQueryToCollection() -> (AddQueryToCollection:bool))
   n3 = Get AssetTagsSubsystem() -> (ReturnValue:Asset Tags Subsystem Object Reference) {is=title K2Node_GetEngineSubsystem}
   n4 = Get CollectionName() -> (CollectionName:Name=None)
   > n5 = CollectionExists!(execute, self:Asset Tags Subsystem Object Reference=@n3, Name:Name=@n4) -> (then, ReturnValue:bool)
   > n6 = Branch(Condition=@n5)
   > n7 = EmptyCollection!(execute, self:Asset Tags Subsystem Object Reference=@n3, Name:Name=@n4) -> (then, ReturnValue:bool)
   n8 = For Each Loop(Exec=[@n7.then, @n2.else, @n6.else], Array:Array of Object References=@n1) -> (Array Element:Object Reference)
   n9 = Length(TargetArray:Array of Object References=@n1)
   > PrintString(InString=Append!(
     A="Operating on Asset "
     B="To String (Integer)"(InInt=@n8.Array Index)
     C:String=" of "
     D:String="To String (Integer)"(InInt=@n9)
     E:String=" : "
     F:String=GetDisplayName(Object=@n8)
    ) -> (ReturnValue))
   > n10 = Cast To StaticMesh(Object=@n8)
   > n11 = GetEditorProperty!(execute:Exec, Object:Object Reference=@n10, PropertyName:Name=NaniteSettings) -> (then:Exec, PropertyValue:"Mesh Nanite Settings Structure (by ref)", ReturnValue:bool)
   > n12 = Branch(Condition=Break Mesh Nanite Settings!(MeshNaniteSettings:"Mesh Nanite Settings Structure (by ref)"=@n11) -> (
     bEnabled:bool=~
     bExplicitTangents:bool=~
     bLerpUVs:bool=~
     bSeparable:bool=~
     bVoxelNDF:bool=~
     bVoxelOpacity:bool=~
     ShapePreservation:ENaniteShapePreservation Enum
     PositionPrecision:int
     NormalPrecision:int
     TangentPrecision:int
     BoneWeightPrecision:int
     KeepPercentTriangles:float=~
     TrimRelativeError:float=~
     GenerateFallback:ENaniteGenerateFallback Enum
     FallbackTarget:ENaniteFallbackTarget Enum
     FallbackPercentTriangles:float=~
     FallbackRelativeError:float=~
     MaxEdgeLengthFactor:float=~
     NumRays:int
     VoxelLevel:int
     RayBackUp:float=~
     DisplacementUVChannel:int
    ) {is=title K2Node_BreakStruct})
   n13 = Get StaticMeshEditorSubsystem()
   n14 = GetLodBuildSettings!(
    execute=[@n27.else, @n26.else, @n12.then]
    self:Static Mesh Editor Subsystem Object Reference=@n13
    StaticMesh:Static Mesh Object Reference=@n10
    LodIndex:int=0
   ) -> (then, OutBuildOptions:Mesh Build Settings Structure)
   > n15 = Branch(Condition=Break Mesh Build Settings!(MeshBuildSettings:"Mesh Build Settings Structure (by ref)"=@n14) -> (bGenerateDistanceFieldAsIfTwoSided:bool=~) {is=title K2Node_BreakStruct})
   > n16 = Branch(Condition=Get OnlyQueryDontChange() -> (OnlyQueryDontChange:bool))
   n17 = Get NumberAffected() -> (NumberAffected:int=0)
   > Increment Int(Value=@n17)
   > Branch(Condition=Get AddQueryToCollection())
   n18 = Get AssetTagsSubsystem()
   n19 = Get CollectionName() -> (CollectionName=None)
   > n20 = CollectionExists(self=@n18, Name=@n19)
   > n21 = Branch(Condition=@n20)
   AddAssetToCollection!(
    execute=[@n22.then, @n21.then]
    self:Asset Tags Subsystem Object Reference=@n18
    Name:Name=@n19
    AssetPathName:Name=String To Name(InString=Get Object Path String(Object=@n10))
   ) -> (then, ReturnValue:bool)
   n22 = CreateCollection!(execute=@n21.else, self:Asset Tags Subsystem Object Reference=@n18, Name:Name=@n19, ShareType:ECollectionScriptingShareType Enum=Local) -> (then, ReturnValue:bool)
   Increment Int(" "=@n16.else, Value=@n17)
   > n23 = Set members in Mesh Build Settings!(execute:Exec, StructRef:"Mesh Build Settings Structure (by ref)"=@n14, bGenerateDistanceFieldAsIfTwoSided:bool=False) -> (then:Exec, StructOut:"Mesh Build Settings Structure (by ref)") {is=title K2Node_SetFieldsInStruct}
   > SetLodBuildSettings!(
    execute
    self:Static Mesh Editor Subsystem Object Reference=@n13
    StaticMesh:Static Mesh Object Reference=@n10
    LodIndex:int=0
    BuildOptions:"Mesh Build Settings Structure (by ref)"=@n23
   ) -> (then)
   > SaveLoadedAsset(AssetToSave=@n8, bOnlyIfIsDirty=true)
   PrintString(execute=@n15.else, InString=Not Twosided)
   Event Construct()
   > SetObject(self=Get ConfigurableDataDV() -> (ConfigurableDataDV:Details View Object Reference), NewObject=Self-Reference())
   Set SafeToConvert?(SafeToConvert?:bool=true) -> (Output_Get:bool)
   > n24 = For Each Loop!(
    Exec
    Array:Array of Static Material Structures=Get StaticMaterials!(self:Static Mesh Object Reference=@n10) -> (StaticMaterials:Array of Static Material Structures)
   ) -> (
    LoopBody
    Array Element_MaterialInterface:Material Interface Object Reference
    Array Element_MaterialSlotName:Name
    Array Element_UVChannelData:Mesh UVChannel Info Structure
    Array Element_OverlayMaterialInterface:Material Interface Object Reference
    Array Index
    Completed
   )
   Branch(execute=@n24.Completed, Condition=Get SafeToConvert?() -> (SafeToConvert?:bool))
   > n25 = CheckForValidReferencers!(
    execute
    self
    StaticMesh:Static Mesh Object Reference=@n10
    ParentClassContains:String=Get Parent Class Contains() -> (Parent Class Contains:String)
   ) -> (then, DoesSM_HaveReferencerOfGivenClass?:bool) {is=title K2Node_CallFunction}
   > Branch(Condition=@n25)
   > n26 = Branch(Condition=IsChecked(self=Get CheckBox_35() -> (CheckBox_35:Check Box Object Reference)))
   > n27 = Branch(Condition=IsLODScreenSizeAutoComputed!(self:Static Mesh Object Reference=@n10) -> (ReturnValue:bool))
   n28 = Switch on EBlendMode!(
    execute=@n24.LoopBody
    Selection:EBlendMode Enum=GetBlendMode!(self:Material Interface Object Reference=@n24) -> (ReturnValue:EBlendMode Enum=BLEND_Opaque)
   ) -> (
    BLEND_Opaque:Exec
    BLEND_Masked:Exec
    BLEND_Translucent:Exec
    BLEND_Additive:Exec
    BLEND_Modulate:Exec
    BLEND_AlphaComposite:Exec
    BLEND_AlphaHoldout:Exec
    BLEND_TranslucentColoredTransmittance:Exec
   ) {is=title K2Node_SwitchEnum}
   Set SafeToConvert?(
    execute=[@n28.BLEND_Additive, @n28.BLEND_AlphaComposite, @n28.BLEND_AlphaHoldout, @n28.BLEND_Masked, @n28.BLEND_Modulate, @n28.BLEND_Translucent]
    SafeToConvert?=false
   )
   Event PreConstruct()
   Event Tick()
   Reroute Node(InputPin:Static Mesh Object Reference=@n10) -> (OutputPin:Static Mesh Object Reference)
  ]}
]
variables [
 AddQueryToCollection {type=boolean}
 CollectionName {type=string, default=1_NaniteConversionList}
 NumberAffected {type=integer}
 OnlyQueryDontChange {type=boolean}
 Parent Class Contains {type=string, default=buildingprop}
 SafeToConvert? {type=boolean}
]
widgets {
 info {rootWidgetClass=&Overlay, widgetCount=7}
 widgets [
  Overlay_34 {widgetClassPath=&Overlay, widget=Overlay_34}
  VerticalBox_0 {
   widgetClassPath=&VerticalBox
   parent=Overlay_34
   widget=VerticalBox_0
   slot {class=OverlaySlot, properties {horizontalAlignment=HAlign_Fill, verticalAlignment=VAlign_Fill}}
  }
  TA_Generic01_Button {
   widgetClassPath=&TA_Generic01_Button
   parent=VerticalBox_0
   widget=TA_Generic01_Button
   properties {label=Convert Selected Nanite Assets to Single Sided}
   slot {class=VerticalBoxSlot}
  }
  HorizontalBox_0 {widgetClassPath=&HorizontalBox, parent=VerticalBox_0, widget=HorizontalBox_0, slot {class=VerticalBoxSlot}}
  CheckBox_35 {
   widgetClassPath=/Script/UMG.CheckBox
   bIsVariable=true
   parent=HorizontalBox_0
   widget=CheckBox_35
   properties {
    checkedState=Checked
    widgetStyle {
     checkBoxType=CheckBox
     backgroundHoveredImage {
      bIsDynamicallyLoaded=false
      drawAs=RoundedBox
      imageType=NoImage
      mirroring=NoMirror
      resourceName=None
      resourceObject=null
      tiling=NoTile
      imageSize=<18 18>
      margin=<ltrb 0 0 0 0>
      outlineSettings {
       bUseBrushTransparency=false
       roundingType=FixedRadius
       width=1
       color {colorUseRule=UseColor_ColorTable, specifiedColor=<rgba 0 0 0 0>}
       cornerRadii=<3 3 3 3>
      }
      tintColor {colorUseRule=UseColor_Specified, specifiedColor=<rgba 0.0048 0.0048 0.0048 1>}
      uVRegion {bIsValid=false, max=<0 0>, min=<0 0>}
     }
     backgroundImage {
      bIsDynamicallyLoaded=false
      drawAs=RoundedBox
      imageType=NoImage
      mirroring=NoMirror
      resourceName=None
      resourceObject=null
      tiling=NoTile
      imageSize=<18 18>
      margin=<ltrb 0 0 0 0>
      outlineSettings {
       bUseBrushTransparency=false
       roundingType=FixedRadius
       width=1
       color {colorUseRule=UseColor_ColorTable, specifiedColor=<rgba 0 0 0 0>}
       cornerRadii=<3 3 3 3>
      }
      tintColor {colorUseRule=UseColor_Specified, specifiedColor=<rgba 0.0048 0.0048 0.0048 1>}
      uVRegion {bIsValid=false, max=<0 0>, min=<0 0>}
     }
     backgroundPressedImage {
      bIsDynamicallyLoaded=false
      drawAs=RoundedBox
      imageType=NoImage
      mirroring=NoMirror
      resourceName=None
      resourceObject=null
      tiling=NoTile
      imageSize=<18 18>
      margin=<ltrb 0 0 0 0>
      outlineSettings {
       bUseBrushTransparency=false
       roundingType=FixedRadius
       width=1
       color {colorUseRule=UseColor_ColorTable, specifiedColor=<rgba 0 0 0 0>}
       cornerRadii=<3 3 3 3>
      }
      tintColor {colorUseRule=UseColor_Specified, specifiedColor=<rgba 0.0048 0.0048 0.0048 1>}
      uVRegion {bIsValid=false, max=<0 0>, min=<0 0>}
     }
     borderBackgroundColor {colorUseRule=UseColor_Specified, specifiedColor=<rgba 1 1 1 1>}
     checkedForeground {colorUseRule=UseColor_Specified, specifiedColor=<rgba 1 1 1 1>}
     checkedHoveredForeground {colorUseRule=UseColor_Specified, specifiedColor=<rgba 1 1 1 1>}
     checkedHoveredImage {
      bIsDynamicallyLoaded=false
      drawAs=Image
      imageType=Vector
      mirroring=NoMirror
      resourceName=../../../Engine/Content/Slate/Starship/CoreWidgets/CheckBox/check.svg
      resourceObject=null
      tiling=NoTile
      imageSize=<16 16>
      margin=<ltrb 0 0 0 0>
      outlineSettings {
       bUseBrushTransparency=false
       roundingType=HalfHeightRadius
       width=0
       color {colorUseRule=UseColor_Specified, specifiedColor=<rgba 0 0 0 0>}
       cornerRadii=<0 0 0 0>
      }
      tintColor {colorUseRule=UseColor_Specified, specifiedColor=<rgba 0.0044 0.2384 1 1>}
      uVRegion {bIsValid=false, max=<0 0>, min=<0 0>}
     }
     checkedImage {
      bIsDynamicallyLoaded=false
      drawAs=Image
      imageType=Vector
      mirroring=NoMirror
      resourceName=../../../Engine/Content/Slate/Starship/CoreWidgets/CheckBox/check.svg
      resourceObject=null
      tiling=NoTile
      imageSize=<16 16>
      margin=<ltrb 0 0 0 0>
      outlineSettings {
       bUseBrushTransparency=false
       roundingType=HalfHeightRadius
       width=0
       color {colorUseRule=UseColor_Specified, specifiedColor=<rgba 0 0 0 0>}
       cornerRadii=<0 0 0 0>
      }
      tintColor {colorUseRule=UseColor_Specified, specifiedColor=<rgba 0 0.162 0.7454 1>}
      uVRegion {bIsValid=false, max=<0 0>, min=<0 0>}
     }
     checkedPressedForeground {colorUseRule=UseColor_Specified, specifiedColor=<rgba 1 1 1 1>}
     checkedPressedImage {
      bIsDynamicallyLoaded=false
      drawAs=Image
      imageType=Vector
      mirroring=NoMirror
      resourceName=../../../Engine/Content/Slate/Starship/CoreWidgets/CheckBox/check.svg
      resourceObject=null
      tiling=NoTile
      imageSize=<16 16>
      margin=<ltrb 0 0 0 0>
      outlineSettings {
       bUseBrushTransparency=false
       roundingType=HalfHeightRadius
       width=0
       color {colorUseRule=UseColor_Specified, specifiedColor=<rgba 0 0 0 0>}
       cornerRadii=<0 0 0 0>
      }
      tintColor {colorUseRule=UseColor_Specified, specifiedColor=<rgba 0 0.0802 0.3515 1>}
      uVRegion {bIsValid=false, max=<0 0>, min=<0 0>}
     }
     checkedSlateSound {resourceObject=null}
     foregroundColor {colorUseRule=UseColor_Specified, specifiedColor=<rgba 1 1 1 1>}
     hoveredForeground {colorUseRule=UseColor_Specified, specifiedColor=<rgba 1 1 1 1>}
     hoveredSlateSound {resourceObject=null}
     padding=<ltrb 2 0 0 0>
     pressedForeground {colorUseRule=UseColor_Specified, specifiedColor=<rgba 1 1 1 1>}
     uncheckedHoveredImage {
      bIsDynamicallyLoaded=false
      drawAs=NoDrawType
      imageType=NoImage
      mirroring=NoMirror
      resourceName=None
      resourceObject=null
      tiling=NoTile
      imageSize=<0 0>
      margin=<ltrb 0 0 0 0>
      outlineSettings {
       bUseBrushTransparency=false
       roundingType=HalfHeightRadius
       width=0
       color {colorUseRule=UseColor_Specified, specifiedColor=<rgba 0 0 0 0>}
       cornerRadii=<0 0 0 0>
      }
      tintColor {colorUseRule=UseColor_Specified, specifiedColor=<rgba 1 1 1 1>}
      uVRegion {bIsValid=false, max=<0 0>, min=<0 0>}
     }
     uncheckedImage {
      bIsDynamicallyLoaded=false
      drawAs=NoDrawType
      imageType=NoImage
      mirroring=NoMirror
      resourceName=None
      resourceObject=null
      tiling=NoTile
      imageSize=<0 0>
      margin=<ltrb 0 0 0 0>
      outlineSettings {
       bUseBrushTransparency=false
       roundingType=HalfHeightRadius
       width=0
       color {colorUseRule=UseColor_Specified, specifiedColor=<rgba 0 0 0 0>}
       cornerRadii=<0 0 0 0>
      }
      tintColor {colorUseRule=UseColor_Specified, specifiedColor=<rgba 1 1 1 1>}
      uVRegion {bIsValid=false, max=<0 0>, min=<0 0>}
     }
     uncheckedPressedImage {
      bIsDynamicallyLoaded=false
      drawAs=NoDrawType
      imageType=NoImage
      mirroring=NoMirror
      resourceName=None
      resourceObject=null
      tiling=NoTile
      imageSize=<0 0>
      margin=<ltrb 0 0 0 0>
      outlineSettings {
       bUseBrushTransparency=false
       roundingType=HalfHeightRadius
       width=0
       color {colorUseRule=UseColor_Specified, specifiedColor=<rgba 0 0 0 0>}
       cornerRadii=<0 0 0 0>
      }
      tintColor {colorUseRule=UseColor_Specified, specifiedColor=<rgba 1 1 1 1>}
      uVRegion {bIsValid=false, max=<0 0>, min=<0 0>}
     }
     uncheckedSlateSound {resourceObject=null}
     undeterminedForeground {colorUseRule=UseColor_Specified, specifiedColor=<rgba 1 1 1 1>}
     undeterminedHoveredImage {
      bIsDynamicallyLoaded=false
      drawAs=Image
      imageType=Vector
      mirroring=NoMirror
      resourceName=../../../Engine/Content/Slate/Starship/CoreWidgets/CheckBox/indeterminate.svg
      resourceObject=null
      tiling=NoTile
      imageSize=<16 16>
      margin=<ltrb 0 0 0 0>
      outlineSettings {
       bUseBrushTransparency=false
       roundingType=HalfHeightRadius
       width=0
       color {colorUseRule=UseColor_Specified, specifiedColor=<rgba 0 0 0 0>}
       cornerRadii=<0 0 0 0>
      }
      tintColor {colorUseRule=UseColor_Specified, specifiedColor=<rgba 0.0044 0.2384 1 1>}
      uVRegion {bIsValid=false, max=<0 0>, min=<0 0>}
     }
     undeterminedImage {
      bIsDynamicallyLoaded=false
      drawAs=Image
      imageType=Vector
      mirroring=NoMirror
      resourceName=../../../Engine/Content/Slate/Starship/CoreWidgets/CheckBox/indeterminate.svg
      resourceObject=null
      tiling=NoTile
      imageSize=<16 16>
      margin=<ltrb 0 0 0 0>
      outlineSettings {
       bUseBrushTransparency=false
       roundingType=HalfHeightRadius
       width=0
       color {colorUseRule=UseColor_Specified, specifiedColor=<rgba 0 0 0 0>}
       cornerRadii=<0 0 0 0>
      }
      tintColor {colorUseRule=UseColor_Specified, specifiedColor=<rgba 0 0.162 0.7454 1>}
      uVRegion {bIsValid=false, max=<0 0>, min=<0 0>}
     }
     undeterminedPressedImage {
      bIsDynamicallyLoaded=false
      drawAs=Image
      imageType=Vector
      mirroring=NoMirror
      resourceName=../../../Engine/Content/Slate/Starship/CoreWidgets/CheckBox/indeterminate.svg
      resourceObject=null
      tiling=NoTile
      imageSize=<16 16>
      margin=<ltrb 0 0 0 0>
      outlineSettings {
       bUseBrushTransparency=false
       roundingType=HalfHeightRadius
       width=0
       color {colorUseRule=UseColor_Specified, specifiedColor=<rgba 0 0 0 0>}
       cornerRadii=<0 0 0 0>
      }
      tintColor {colorUseRule=UseColor_Specified, specifiedColor=<rgba 0 0.0802 0.3515 1>}
      uVRegion {bIsValid=false, max=<0 0>, min=<0 0>}
     }
    }
   }
   slot {class=HorizontalBoxSlot}
  }
  CommonTextBlock_51 {
   widgetClassPath=&CommonTextBlock
   parent=HorizontalBox_0
   widget=CommonTextBlock_51
   properties {
    text=Modify only Non-Auto LOD Distance Assets
    font {
     bForceMonospaced=false
     bMaterialIsStencil=false
     fontMaterial=null
     fontObject=&NotoSans
     letterSpacing=0
     monospacedWidth=1
     size=12
     skewAmount=0
     typefaceFontName=NotoSans-Regular
     outlineSettings {
      bApplyOutlineToDropShadows=false
      bMiteredCorners=false
      bSeparateFillAlpha=false
      outlineMaterial=null
      outlineSize=0
      outlineColor=<rgba 0 0 0 1>
     }
    }
    shadowOffset=<0 0>
   }
   slot {class=HorizontalBoxSlot}
  }
  ConfigurableDataDV {
   widgetClassPath=&DetailsView
   parent=VerticalBox_0
   widget=ConfigurableDataDV
   properties {categoriesToShow [NoCategory]}
   slot {class=VerticalBoxSlot, properties {horizontalAlignment=HAlign_Left, verticalAlignment=VAlign_Top}}
  }
 ]
}
