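# Editor utility widget that exports a named asset collection to a CSV file and imports one back, using Import and Export buttons.
# A details view bound to the widget exposes its settings, including collection name and CSV path.
# Export gathers each asset's unique class path string (not its object path), then writes them through embedded Python.
# Import reads rows through Python, empties the collection or creates a local one, then adds each entry as an asset path.
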EditorUtilityWidgetBlueprint /Game/ArtTools/Tech_Art/Nanite_Tools/EUW_ImportExport_Collection
graphs [EventGraph {nodes [
   "On Clicked (Export)"!() -> (OutputDelegate, then, widget:User Widget Object Reference, event:Pointer Event Structure) {is=title K2Node_ComponentBoundEvent}
   > Clear(TargetArray=Get ObjectPaths() -> (ObjectPaths:Array of Strings))
   n1 = Get AssetTagsSubsystem() -> (ReturnValue:Asset Tags Subsystem Object Reference) {is=title K2Node_GetEngineSubsystem}
   > n2 = GetAssetsInCollection!(execute, self:Asset Tags Subsystem Object Reference=@n1, Name:Name=Get Collection() -> (Collection:Name=None)) -> (then, ReturnValue:Array of Asset Data Structures)
   > n3 = For Each Loop!(Exec, Array:Array of Asset Data Structures=@n2) -> (
    LoopBody
    Array Element_PackageName:Name
    Array Element_PackagePath:Name
    Array Element_AssetName:Name
    Array Element_AssetClassPath_PathString:String
    Array Index
    Completed
   )
   n4 = Get ObjectPaths()
   Execute Python Script!(
    execute=@n3.Completed
    PythonScript="import csv\r\ncsvfile = open(filepath, 'w', newline = '')\r\n\r\nwriter = csv.writer(csvfile)\r\n\r\nfor line in rows:\r\n    values = line.split(',')\r\n    print(values)\r\n    writer.writerow(values)\r\n"
    Rows:Array of Strings=@n4
    filepath:String=Get!() -> (CSV Path_FilePath:String)
   ) -> (then, ReturnValue) {inputs [Rows,filepath]}
   Add Unique(execute=@n3.LoopBody, TargetArray=@n4, NewItem=@n3.Array Element_AssetClassPath_PathString)
   "On Clicked (Import)"!() -> (OutputDelegate, then, widget:User Widget Object Reference, event:Pointer Event Structure) {is=title K2Node_ComponentBoundEvent}
   > n5 = Execute Python Script!(
    execute
    PythonScript="import csv\r\n\r\nlines = []\r\n\r\nwith open(filepath, newline ='') as csvfile:\r\n    spamreader = csv.reader(csvfile, quotechar='|')\r\n    for row in spamreader:\r\n        print(row)\r\n        ##section added to remove [' and '] from string\r\n        tempstr = str(row)\r\n        tempstr = tempstr[2:len(tempstr)-2]\r\n        print(tempstr)\r\n        lines.append(tempstr)"
    Filepath:String=Break File Path!(FilePath:"File Path Structure (by ref)"=Get CSV Path() -> (CSV Path:File Path Structure)) -> (FilePath:String) {is=title K2Node_BreakStruct}
   ) -> (then, ReturnValue, Lines:Array of Strings) {inputs [Filepath], outputs [Lines]}
   > n6 = Set ObjectPaths(ObjectPaths:Array of Strings=@n5.Lines) -> (Output_Get:Array of Strings)
   n7 = Get Collection() -> (Collection=None)
   > n8 = CollectionExists!(execute, self:Asset Tags Subsystem Object Reference=@n1, Name:Name=@n7) -> (then, ReturnValue:bool)
   > n9 = Branch(Condition=@n8)
   > n10 = EmptyCollection!(execute, self:Asset Tags Subsystem Object Reference=@n1, Name:Name=@n7) -> (then, ReturnValue:bool)
   n11 = For Each Loop!(Exec=[@n10.then, @n12.then], Array:Array of Strings=@n6) -> (LoopBody, Array Element:String, Array Index, Completed)
   > AddAssetToCollection!(execute, self:Asset Tags Subsystem Object Reference=@n1, Name:Name=@n7, AssetPathName:Name=String To Name(InString=@n11)) -> (then, ReturnValue:bool)
   n12 = CreateCollection!(execute=@n9.else, self:Asset Tags Subsystem Object Reference=@n1, Name:Name=@n7, ShareType:ECollectionScriptingShareType Enum=Local) -> (then, ReturnValue:bool)
   Event Construct()
   > SetObject(self=Get DetailsView_70() -> (DetailsView_70:Details View Object Reference), NewObject=Self-Reference())
   Event PreConstruct()
   Event Tick()
  ]}]
variables [
 Collection {type=string, default=None}
 CSV Path {type=FilePath, default {filePath=""}}
 ObjectPaths {type="array<string>", default []}
]
widgets {
 info {rootWidgetClass=&Overlay, widgetCount=6}
 widgets [
  Overlay_16 {widgetClassPath=&Overlay, widget=Overlay_16}
  VerticalBox_0 {
   widgetClassPath=&VerticalBox
   parent=Overlay_16
   widget=VerticalBox_0
   slot {class=OverlaySlot, properties {horizontalAlignment=HAlign_Fill, verticalAlignment=VAlign_Fill}}
  }
  HorizontalBox_28 {widgetClassPath=&HorizontalBox, parent=VerticalBox_0, widget=HorizontalBox_28, slot {class=VerticalBoxSlot}}
  Import {
   widgetClassPath=&TA_Generic01_Button
   parent=HorizontalBox_28
   widget=Import
   properties {label=Import}
   slot {class=HorizontalBoxSlot, properties {size {sizeRule=Fill, value=1}}}
  }
  Export {
   widgetClassPath=&TA_Generic01_Button
   parent=HorizontalBox_28
   widget=Export
   properties {label=Export}
   slot {class=HorizontalBoxSlot, properties {size {sizeRule=Fill, value=1}}}
  }
  DetailsView_70 {
   widgetClassPath=&DetailsView
   parent=VerticalBox_0
   widget=DetailsView_70
   properties {categoriesToShow [NoCategory]}
   slot {class=VerticalBoxSlot}
  }
 ]
}
